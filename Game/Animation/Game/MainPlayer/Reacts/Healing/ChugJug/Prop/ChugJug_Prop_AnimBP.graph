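# Animation Blueprint for the Chug Jug prop, built on the Chug Jug gadget skeleton.
# A looping-capable sequence player feeds the ConsumableLarge default prop pose through DefaultSlot to the output, so montages can override it.
# A root bone transform node with a 90-degree yaw replace exists but is not connected to the output.
# The event graph holds only unwired default nodes, and there are no variables.
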AnimBlueprint /Game/Animation/Game/MainPlayer/Reacts/Healing/ChugJug/Prop/ChugJug_Prop_AnimBP
target_skeleton=/Game/Athena/Items/Consumables/SuperMedkit/Chug_Jug_Gadget_Skeleton
graphs [
 AnimGraph {nodes [
   n1 = "Transform (Modify) Bone - Bone: root"!(ComponentPose=Local To Component(), Alpha, Rotation="0,90,0") -> (Pose) {rotationMode=BMM_Replace, rotationSpace=BCS_BoneSpace, alphaBoolBlend {blendOption=Linear}, boneToModify {boneName=root}, rotation {yaw=90}}
   Component To Local(ComponentPose=@n1)
   Output Pose(Result=Slot 'DefaultSlot'(Source=Sequence Player 'ConsumableLarge_Default_Pose_Prop'() {sequence=&/ConsumableLarge_Default_Pose_Prop})) {layerGroup=None}
  ]}
 EventGraph {nodes [
   TryGetPawnOwner()
   Event BlueprintUpdateAnimation()
  ]}
]
variables []
